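; Trajectory preview for the Playset Grenade that draws the throw arc and a grid-snapped box where the playset lands.
; Trajectory updates rebuild the spline meshes, scaling tangents by point spacing to prevent loops, and feed endpoints to the material.
; HandleGridSnapUpdates snaps the endpoint to the building grid, adjusts it through the controller's cached playset, and rotates by instigator cardinal facing.
; UpdateScaleForPlayset stretches the preview mesh to playset size.
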(event EventBeginPlay
  (|CallParentFunction))

(event Collision|EventActorBeginOverlap
  (params (OtherActor :type Actor%20Object%20Reference))
  (|CallParentFunction OtherActor))

(event EventTick
  (params (DeltaSeconds :type Float%20%28single-precision%29))
  (|CallParentFunction DeltaSeconds) ; This node is disabled and will not be called.\nDrag off pins to build functionality.
)

(event Trajectory|EventSetTrajectorySpline
  (params (SplinePoints :type Array%20of%20Vectors) (SplineTangents :type Array%20of%20Vectors))
  (bind _returnvalue (Utilities|Array|LastIndex SplinePoints))
  (bind _output (Utilities|Array|Get(aref) SplinePoints _returnvalue))
  (bind _target (Variables|Default|GetTarget))
  (Rendering|Material|SetVectorParameterValue (Variables|VFX|GetSplineMID) "StartLocation" (Math|Conversions|ToLinearColor(Vector) (Utilities|Array|Get(aref) SplinePoints)) :!class /Script/Engine.MaterialInstanceDynamic)
  (Rendering|Material|SetVectorParameterValue (Variables|VFX|GetSplineMID) "EndLocation" (Math|Conversions|ToLinearColor(Vector) _output) :!class /Script/Engine.MaterialInstanceDynamic)
  (CallFunction|HandleGridSnapUpdates _output)
  (Transformation|SetWorldLocation _target _output false true)
  (Rendering|SetVisibility _target true true)
  (Spline|SetSplinePoints (Variables|Trajectory|GetSplineComponent) SplinePoints "World" true)
  (for (_array_element _array_index) SplinePoints
    ;[1 0.5129 0] For some reason, splines freak out and loop onto themselves when tangent magnitudes are too big. Here we scale them down by the distance between spline points, because that seems to yield reasonably good results
    (if (Utilities|Array|IsValidIndex SplinePoints (+ _array_index 1))
      (Variables|Default|SetLastPointDist (Math|Vector|VectorLength (- _array_element (Utilities|Array|Get(acopy) SplinePoints (+ _array_index 1))))))
    (Spline|SetTangentsatSplinePoint (Variables|Trajectory|GetSplineComponent) _array_index (* (Math|Vector|Normalize (Utilities|Array|Get(acopy) SplineTangents (* _array_index 2))) (Variables|Default|GetLastPointDist)) (* (Math|Vector|Normalize (Utilities|Array|Get(acopy) SplineTangents (+ (* _array_index 2) 1))) (Variables|Default|GetLastPointDist)) "World" true))
  (for (_array_element_2 _array_index_2) (Variables|Trajectory|GetSplineMeshComponents)
    (if (< _array_index_2 (Utilities|Array|LastIndex SplinePoints))
      (Development|SetHiddeninGame _array_element_2 false false)
      (SplineMesh|SetStartandEnd _array_element_2 (Spline|GetLocationatSplinePoint (Variables|Trajectory|GetSplineComponent) _array_index_2 "World") (Spline|GetLeaveTangentatSplinePoint (Variables|Trajectory|GetSplineComponent) _array_index_2 "World") (Spline|GetLocationatSplinePoint (Variables|Trajectory|GetSplineComponent) (+ _array_index_2 1) "World") (Spline|GetArriveTangentatSplinePoint (Variables|Trajectory|GetSplineComponent) (+ _array_index_2 1) "World") true :!class /Script/Engine.SplineMeshComponent)
      (else
        (Development|SetHiddeninGame _array_element_2 true false)))))

(fn GetPlaysetFromController
  (bind _asfort_player_controller_athena (Utilities|Casting|CastToFortPlayerControllerAthena (Game|GetPlayerController 0))
    (:then
      (bind _output_get (Variables|Default|SetCachedPlayset (Creative|GetPlayset _asfort_player_controller_athena)))
      (return (Utilities|IsValid _output_get :!class /Script/Engine.KismetSystemLibrary)))
    (:CastFailed
      (return false))))

(fn HandleGridSnapUpdates
  (params (NewLocation :type Vector))
  (bind _gridsnapposition (LocalVariables|GetGridSnapPosition))
  (Utilities|IsValid (Variables|Default|GetCachedPlayset)
    (:Is%20Valid
      _)
    (:Is%20Not%20Valid
      (bind _bsuccess (CallFunction|GetPlaysetfromController))
      (if _bsuccess
        _)))
  (bind _output_get (LocalVariables|SetGridSnapPosition (Fortnite|SnaptoFortBuildingGrid NewLocation)))
  (bind (_outcardinaldirection _outyaw) (Class|AthenaConsumablesSharedFunctionLibrary|GetActorForwardCardinalDirection (Game|GetInstigator)))
  (LocalVariables|SetYawOffset _outyaw)
  (if (or (Math|Vector|NotEqual(Vector) _output_get (Variables|Default|GetLastGridSnapPosition) 0.0001) (Math|Float|NearlyEqual(Float) _outyaw (Variables|Default|GetLastYawOffset) 0.0001))
    (bind _returnvalue (Playspace|AdjusttoFinalLocation (Variables|Default|GetCachedPlayset) _gridsnapposition (Math|Rotator|CombineRotators (Variables|/Script/FortniteGame.FortPlaysetItemDefinition|GetDefaultRotation (Variables|Default|GetCachedPlayset)) (Math|Rotator|MakeRotator 0.0 0.0 (LocalVariables|GetYawOffset)))))
    (LocalVariables|SetAdjustedPosition _returnvalue)
    (Variables|Default|SetLastGridSnapPosition _gridsnapposition)
    (Variables|Default|SetLastYawOffset (LocalVariables|GetYawOffset))
    (Audio|PlaySound2D (Variables|Default|GetGridSnapSound))
    (Transformation|SetWorldLocation (Variables|Default|GetBuilding) (LocalVariables|GetAdjustedPosition) false true)
    (Transformation|SetWorldRotation (Variables|Default|GetSM_BuildingPreview) "0, 0, 0" false false)
    (CallFunction|UpdateScaleforPlayset)
    (Transformation|SetWorldRotation (Variables|Default|GetSM_BuildingPreview) (Math|Rotator|CombineRotators (Variables|/Script/FortniteGame.FortPlaysetItemDefinition|GetDefaultRotation (Variables|Default|GetCachedPlayset)) (Math|Rotator|MakeRotator 0.0 0.0 (LocalVariables|GetYawOffset))) false false)))

(fn UpdateScaleForPlayset
  (bind (_origin _boxextent) (Collision|GetComponentBounds (Variables|Default|GetSM_BuildingPreview)))
  (bind _sm_buildingpreview (Variables|Default|GetSM_BuildingPreview))
  (bind _asfort_player_controller_athena (Utilities|Casting|CastToFortPlayerControllerAthena (Game|GetPlayerController 0)))
  (Variables|Default|SetCachedPlayset (Creative|GetPlayset _asfort_player_controller_athena))
  (Transformation|SetWorldScale3D _sm_buildingpreview (Math|Vector|MakeVector (.x (/ (* (Math|Vector|MakeVector (* 512.0 (/ (Math|Conversions|ToFloat(Integer) (Creative|GetPlaysetSize _asfort_player_controller_athena)) 2.0)) (* 512.0 (/ (Math|Conversions|ToFloat(Integer) (.SizeY (Creative|GetPlaysetSize _asfort_player_controller_athena))) 2.0)) (* 512.0 (/ (Math|Conversions|ToFloat(Integer) (.SizeZ (Creative|GetPlaysetSize _asfort_player_controller_athena))) 2.0))) (Transformation|GetWorldScale (Variables|Default|GetSM_BuildingPreview))) _boxextent)) (.y (/ (* (Math|Vector|MakeVector (* 512.0 (/ (Math|Conversions|ToFloat(Integer) (Creative|GetPlaysetSize _asfort_player_controller_athena)) 2.0)) (* 512.0 (/ (Math|Conversions|ToFloat(Integer) (.SizeY (Creative|GetPlaysetSize _asfort_player_controller_athena))) 2.0)) (* 512.0 (/ (Math|Conversions|ToFloat(Integer) (.SizeZ (Creative|GetPlaysetSize _asfort_player_controller_athena))) 2.0))) (Transformation|GetWorldScale (Variables|Default|GetSM_BuildingPreview))) _boxextent)) (.z (/ (* (Math|Vector|MakeVector (* 512.0 (/ (Math|Conversions|ToFloat(Integer) (Creative|GetPlaysetSize _asfort_player_controller_athena)) 2.0)) (* 512.0 (/ (Math|Conversions|ToFloat(Integer) (.SizeY (Creative|GetPlaysetSize _asfort_player_controller_athena))) 2.0)) (* 512.0 (/ (Math|Conversions|ToFloat(Integer) (.SizeZ (Creative|GetPlaysetSize _asfort_player_controller_athena))) 2.0))) (Transformation|GetWorldScale (Variables|Default|GetSM_BuildingPreview))) _boxextent))))
  (Rendering|SetVisibility _sm_buildingpreview true true))

(fn ConstructionScript
  (|CallParentFunction)
  (Components|StaticMesh|SetStaticMesh (Variables|Default|GetSM_BuildingPreview) (Variables|Default|GetBuildingPreviewMesh)))
