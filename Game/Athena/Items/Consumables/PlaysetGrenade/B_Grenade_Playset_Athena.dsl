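; Athena weapon Blueprint for the Playset Grenade consumable, handling its input mode on unequip.
; On local unequip it casts the instigator controller to a Fort player controller and pops the playset grenade input mode using the Athena playset grenade input.
; The equipped-and-attached event only performs the same controller cast, with no matching push shown.
; BeginPlay and the construction script just call the parent.
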(event EventBeginPlay
  (|CallParentFunction))

(event FortWeapon|EventOnLocalUnEquip
  (bind _asfort_player_controller (Utilities|Casting|CastToFortPlayerController (Game|GetInstigatorController)))
  (PlaysetGreande|PopPlaysetGrenadeInputMode (Variables|Default|GetAthenaPlaysetGrenadeInput) _asfort_player_controller))

(event FortWeapon|EventOnEquippedAndAttached
  (bind _asfort_player_controller (Utilities|Casting|CastToFortPlayerController (Game|GetInstigatorController))))

(fn ConstructionScript
  (|CallParentFunction))
