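; Playset grenade projectile that deploys a creative playset (prefab) where it lands.
; Begin play derives a cardinal facing from the instigator's forward vector; the first hit stops the projectile without bouncing.
; On stop it caches a grid-snapped destination, aligns itself and explodes.
; The explosion plays effects where cosmetics run; otherwise it adjusts the location and spawns the cached playset through the Volume Manager, subject to the player's creation permission.
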(event EventBeginPlay
  ;[1 1 1] Lookup build delay row. Determine which cardinal direction the tower should face.
  (Utilities|IsValid (Game|GetInstigator)
    (:Is%20Valid
      (bind _outxy (DataTable|EvaluateCurveTableRow "/Game/Athena/Balance/DataTables/AthenaGameData.AthenaGameData" "Default.TowerGrenade.BuildDelay")
        (:RowFound
          (Variables|Default|SetBuildDelay _outxy))
        (:RowNotFound
          _))
      (CallFunction|CalculateCardinalDirection))))

(event EventEndPlay
  (params (EndPlayReason :type EEndPlayReason%20Enum :default "Destroyed"))
  ;[1 1 1] Destroy any remaining actors in our destroy list if needed
  (switch Utilities|FlowControl|Switch|SwitchonEEndPlayReason EndPlayReason))

(event Collision|EventHit
  (params (MyComp :type Primitive%20Component%20Object%20Reference) (Other :type Actor%20Object%20Reference) (OtherComp :type Primitive%20Component%20Object%20Reference) (bSelfMoved :type Boolean) (HitLocation :type Vector :default "0, 0, 0") (HitNormal :type Vector :default "0, 0, 0") (NormalImpulse :type Vector :default "0, 0, 0") (Hit :type Hit%20Result%20Structure%20%28by%20ref%29))
  ;[1 1 1] Do Not Bounce
  (Utilities|FlowControl|DoOnce false
    (:Completed
      (Game|Components|ProjectileMovement|StopSimulating (Variables|Movement|GetProjectileMovementComponent) Hit))))

(event EventOnStop
  (params (Hit :type Hit%20Result%20Structure%20%28by%20ref%29))
  ;[1 1 1] Initializing variables, stop rotation, and snap projectile to grid before "exploding"
  (CallFunction|SetBaseDestination Hit)
  (Components|Movement|SetUpdatedComponent (Variables|Default|GetRotatingMovement) 0)
  (Transformation|SetActorTransform self (Math|Transform|MakeTransform (Fortnite|SnaptoFortBuildingGrid (Transformation|GetActorLocation)) (Math|Rotator|MakeRotator 0.0 0.0 (Utilities|Map|Find (Variables|Orientation|GetOrientationYawOffsets) (Variables|Orientation|GetCardinalDirection)))) false false)
  (Delivery|DoExplosion)
  (bind _object (Utilities|AsyncLoadAsset "/Game/Sounds/Fort_Gadgets/PopupGrenade/Creative_Grenade_Explode_Cue.Creative_Grenade_Explode_Cue")
    (:Completed
      (bind _assound_base (Utilities|Casting|CastToSoundBase _object))
      (Audio|PlaySoundatLocation _assound_base (Transformation|GetActorLocation)))))

(event EventOnExploded
  (params (HitActors :type Array%20of%20Actor%20Object%20References) (HitResults :type Array%20of%20Hit%20Result%20Structures))
  (bind _cardinaldirection (Variables|Orientation|GetCardinalDirection))
  (bind _orientationyawoffsets (Variables|Orientation|GetOrientationYawOffsets))
  (bind _value (Utilities|Map|Find _orientationyawoffsets _cardinaldirection))
  (bind _returnvalue (Math|Rotator|MakeRotator 0.0 0.0 _value))
  (bind _returnvalue_2 (Transformation|GetActorLocation))
  (bind _cachedplayset (Variables|Default|GetCachedPlayset))
  (bind _defaultrotation (Variables|/Script/FortniteGame.FortPlaysetItemDefinition|GetDefaultRotation _cachedplayset))
  (bind _returnvalue_3 (Math|Rotator|CombineRotators _defaultrotation _returnvalue))
  (bind _popupgrenade_throwloop_audiocomponent (Variables|Default|GetPopupGrenade_ThrowLoop_AudioComponent)
    (:then
      (Audio|Components|Audio|FadeOut _popupgrenade_throwloop_audiocomponent 0.2 0.0 "Linear" :!class /Script/Engine.AudioComponent))
    (:else
      _))
  (Utilities|IsValid _cachedplayset
    (:Is%20Valid
      (Networking|CanExecuteCosmeticEvents
        (:True
          ;[1 1 1] OnExploded - Emitter, Sound
          (Effects|SpawnEmitteratLocation (Variables|Cosmetic|GetP_BuildBuilding) _returnvalue_2)
          (Camera|PlayWorldCameraShake "/Game/Athena/Items/Consumables/TowerGrenade/FX/B_Ground_CameraShake.B_Ground_CameraShake_C" (- _returnvalue_2 "0, 0,500.000000") 1.5e+02 2e+03 2.0))
        (:False
          (bind _returnvalue_4 (Playspace|AdjusttoFinalLocation _cachedplayset (Variables|Default|GetBaseDestination) _returnvalue_3))
          (CallFunction|SpawnPlayspaceServer _cachedplayset _returnvalue_4 _returnvalue_3)
          (CallFunction|PlaySpawnPrefabSound _returnvalue_4))))))

(event Custom|Play%20Spawn%20Prefab%20Sound
  (params (Location :type Vector))
  (Audio|PlaySoundatLocation "/Game/Sounds/Creative/Gadgets/Building/Building_Prefab_Spawn_Cue.Building_Prefab_Spawn_Cue" Location))

(event Custom|SpawnPlayspaceServer
  (params (Playset :type Fort%20Playset%20Item%20Definition%20Object%20Reference) (Location :type Vector) (Rotation :type Rotator))
  ;[1 1 1] Server Spawning the Playspace
  (bind _asfort_player_controller_athena (Utilities|Casting|CastToFortPlayerControllerAthena (Pawn|GetController (Game|GetInstigator))))
  (if (.ReturnValue (Creative|CanPlayerCreateinCurrentVolume _asfort_player_controller_athena :!class /Script/FortniteGame.FortPlayerControllerAthena))
    (bind _outvolumemanager (Volume|GetVolumeManager)
      (:Valid
        (VolumeManager|SpawnVolume _outvolumemanager (Variables|Default|GetPlayspaceActorToSpawn) (Creative|CanPlayerCreateinCurrentVolume _asfort_player_controller_athena :!class /Script/FortniteGame.FortPlayerControllerAthena) Playset Location Rotation)))))

(fn CalculateCardinalDirection
  (bind _returnvalue (Game|GetInstigator))
  (bind _returnvalue_2 (Transformation|GetActorForwardVector _returnvalue))
  (bind _x (.x _returnvalue_2))
  (bind _y (.y _returnvalue_2))
  (if (>= (Math|Float|Absolute(Float) _x) (Math|Float|Absolute(Float) _y))
    (if (>= _x 0.0)
      (Variables|Orientation|SetCardinalDirection "NewEnumerator0")
      (else
        (Variables|Orientation|SetCardinalDirection "NewEnumerator1")))
    (elif (>= _y 0.0)
      (Variables|Orientation|SetCardinalDirection "NewEnumerator2")
      (else
        (Variables|Orientation|SetCardinalDirection "NewEnumerator3")))))

(fn SetBaseDestination
  (params (Hit :type Hit%20Result%20Structure%20%28by%20ref%29))
  (bind (_bblockinghit _binitialoverlap _time _distance _location _impactpoint) (Collision|BreakHitResult Hit))
  (Networking|SwitchHasAuthority
    (:Authority
      (Variables|Default|SetBaseDestination (Fortnite|SnaptoFortBuildingGrid _impactpoint)))))

(fn SetPlaysetData
  (params (NewPlayset :type Fort%20Playset%20Item%20Definition%20Object%20Reference))
  (Variables|Default|SetCachedPlayset NewPlayset))

(fn ConstructionScript
  (Networking|CanExecuteCosmeticEvents
    (:True
      (Components|Movement|SetUpdatedComponent (Variables|Default|GetRotatingMovement) (Variables|Default|GetMesh)))))
